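# Niagara trail effect for the Caveman male pickaxe, a short-lived, faint pale-blue ribbon drawn during animation.
# One emitter uses shared AnimTrail modules to spawn and update ribbon particles from bone or socket indices, with a skeletal mesh user parameter exposed.
# Material erosion and UVs follow normalized loop age, while alpha scales by time of day.
# Curves fade the color and taper the ribbon width.
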NiagaraSystem /BRCosmetics/Weapons/FORT_Melee/Pickaxe_Caveman_Male/FX/NS_Pickaxe_Caveman_AnimTrail
dependencies {usedDataInterfaces [&NiagaraDataInterfaceCurve], usedDynamicInputs [], usedModules [&SystemState]}
emitters {Ribbon_AnimTrail {
  emitterUpdateScript [
   &EmitterState()
   &FN_SpawnAnimTrail_v2()
  ]
  particleSpawnScript [
   /Game/Effects/Niagara/Modules/AnimTrail/FN_InitializeAnimTrail/FN_InitializeAnimTrailBoneIndecies(First Bone or Socket Index=1)
   /Game/Effects/Niagara/Modules/AnimTrail/FN_InitializeAnimTrail/FN_InitializeAnimTrail(AnimTrailWidth=0.7, Color=<rgba 0.8 1 1.2 0.25>, Lifetime=0.25) {module=FN_InitializeAnimTrail001}
  ]
  particleUpdateScript [
   &ParticleState()
   /Game/Effects/Niagara/Modules/AnimTrail/FN_UpdateAnimTrail/FN_UpdateAnimTrail_v3/FN_UpdateAnimTrail()
   &FN_RibbonMaterial(
    Color=<rgba 0.7496 0.9293 1 1>
    Erosion Amount:NiagaraFloat=&Lerp_Float(
     A=0
     B=1
     Alpha=&FloatFromCurve(
      FloatCurve=DataInterface() {propertyValues {Curve {keys [<at 0 1 RCIM_Cubic>,<at 0.1 0 RCIM_Cubic>,<at 0.5 1>]}}}
      CurveIndex=$System.NormalizedLoopAge
     )
    )
    Erosion Power=0
    Scale Alpha:NiagaraFloat=&FNTimeOfDayScalar(Night Scale=0.55, Morning Scale=0.6, Day Scale=0.8, Evening Scale=0.6)
    UV Basis:NiagaraFloat=$System.NormalizedLoopAge
   )
   &ScaleColor(
    Color Value To Scale=$Particles.Color
    Scale Alpha=&FloatFromCurve(FloatCurve=DataInterface() {propertyValues {Curve {keys [<at 0 0.5>,<at 1 0>]}}})
    ScaleA=true
    ScaleRGB=true
   )
   &ScaleRibbonWidth(Ribbon Width Scale=&FloatFromCurve(FloatCurve=DataInterface() {propertyValues {Curve {keys [
         {arriveTangent=-1.6295, interpMode=RCIM_Cubic, leaveTangent=-1.0144, tangentMode=RCTM_User, time=0, value=1}
         {arriveTangent=-1, interpMode=RCIM_Cubic, leaveTangent=-1, time=0.5, value=0.3515}
         {arriveTangent=-0.3934, interpMode=RCIM_Cubic, leaveTangent=-0.3934, tangentMode=RCTM_User, time=1, value=0}
        ]}}}))
  ]
  rendererClasses [&NiagaraRibbonRendererProperties]
  renderers [{rendererClass=&NiagaraRibbonRendererProperties}]
 }}
summary {
 systemName=NS_Pickaxe_Caveman_AnimTrail
 emitters [{emitterName=Ribbon_AnimTrail, rendererClasses [&NiagaraRibbonRendererProperties]}]
 userVariables [
  User.LifetimeCustom {type=NiagaraFloat, defaultValue=0, description=""}
  User.RibbonIDOffset {type=NiagaraInt32}
  User.SkeletalMesh {
   type=NiagaraDataInterfaceSkeletalMesh
   defaultValue {
    struct=NiagaraExt_VariableValue_DataInterface
    value {dataInterface=&self:NiagaraDataInterfaceSkeletalMesh_0}
   }
  }
 ]
}
