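# Physics asset for the Caveman Male pickaxe, a two-handed melee cosmetic with four dangling fish that can swing as secondary motion.
# One long capsule covers the handle, and each fish has head, tail, top fin and bottom fin capsules on dyn_ bones.
# Each head and tail is constrained separately to the handle with its own swing and twist limits.
# Both fins constrain to their fish's head with uniformly tighter limits.
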PhysicsAsset /BRCosmetics/Weapons/FORT_Melee/Pickaxe_Caveman_Male/Meshes/Caveman_Male_Axe_Physics
bodies {
 Melee_TwoHanded_Handle [{shapeType=Capsule, length=170.5912, radius=5.5348, center=<0.0133 -1.1108 43.3323>, rotation=<rot -0.4058 -90 0.0145>}]
 dyn_fish01_bFin [{shapeType=Capsule, length=6.2027, radius=3.1773, center=<-3.5955 4.4124 -2.0859>, rotation=<rot 0 50.5042 -73.1679>}]
 dyn_fish01_head [{shapeType=Capsule, length=5.755, radius=3.88, center=<11.6757 0.0001 0.8526>, rotation=<rot 0 -89.9998 89.9996>}]
 dyn_fish01_tFin [{shapeType=Capsule, length=6.6437, radius=3.2617, center=<-3.5445 5.0708 1.3641>, rotation=<rot 0 -136.4795 -81.6377>}]
 dyn_fish01_tail [{shapeType=Capsule, length=7.125, radius=2.375, center=<10.4009 0.0001 0.0002>, rotation=<rot 0 -89.9999 89.9989>}]
 dyn_fish02_bFin [{shapeType=Capsule, length=6.3014, radius=2.9299, center=<-3.9212 -4.6942 -0.547>, rotation=<rot 0 128.8812 -86.9768>}]
 dyn_fish02_head [{shapeType=Capsule, length=5.255, radius=4.38, center=<7.2543 0.0002 0>, rotation=<rot 0 -89.9999 89.9992>}]
 dyn_fish02_tFin [{shapeType=Capsule, length=6.5834, radius=3.4956, center=<-4.0621 -4.4108 1.9771>, rotation=<rot 0 -51.2941 -77.024>}]
 dyn_fish02_tail [{shapeType=Capsule, length=6.625, radius=2.875, center=<11.6522 0.0003 -0.0003>, rotation=<rot 0 -89.9999 90.0017>}]
 dyn_fish03_bFin [{shapeType=Capsule, length=6.0942, radius=3.1218, center=<-3.1476 3.2925 -3.8591>, rotation=<rot 0 58.7338 -52.363>}]
 dyn_fish03_head [{shapeType=Capsule, length=5.005, radius=4.63, center=<8.0051 0 -0.0001>, rotation=<rot 0 -89.9997 90.0035>}]
 dyn_fish03_tFin [{shapeType=Capsule, length=6.321, radius=3.3206, center=<-4.5051 5.0467 -1.1641>, rotation=<rot 0 51.3844 -75.1274>}]
 dyn_fish03_tail [{shapeType=Capsule, length=7.125, radius=2.375, center=<5.6038 0.0001 -0.0001>, rotation=<rot 0 -89.9998 90.0017>}]
 dyn_fish04_bFin [{shapeType=Capsule, length=5.9814, radius=2.4343, center=<-2.6864 -5.0392 -1.1633>, rotation=<rot 0 139.7073 -76.4617>}]
 dyn_fish04_head [{shapeType=Capsule, length=7.505, radius=4.38, center=<8.0344 -0.0001 0>, rotation=<rot 0 -89.9998 89.9996>}]
 dyn_fish04_tFin [{shapeType=Capsule, length=6.4674, radius=3.2799, center=<-3.7266 -4.3129 2.4361>, rotation=<rot 0 -44.8968 -75.3815>}]
 dyn_fish04_tail [{shapeType=Capsule, length=7.125, radius=2.375, center=<4.8654 -0.0001 -0.0001>, rotation=<rot 0 -89.9998 90.0017>}]
}
constraints [
 {bone1Name=dyn_fish01_tail, bone2Name=Melee_TwoHanded_Handle, swing1LimitDegrees=60, swing2LimitDegrees=60, twistLimitDegrees=9}
 {bone1Name=dyn_fish01_head, bone2Name=Melee_TwoHanded_Handle, swing1LimitDegrees=50, swing2LimitDegrees=45, twistLimitDegrees=10}
 {bone1Name=dyn_fish01_tFin, bone2Name=dyn_fish01_head, swing1LimitDegrees=15, swing2LimitDegrees=15, twistLimitDegrees=5}
 {bone1Name=dyn_fish01_bFin, bone2Name=dyn_fish01_head, swing1LimitDegrees=15, swing2LimitDegrees=15, twistLimitDegrees=5}
 {bone1Name=dyn_fish02_head, bone2Name=Melee_TwoHanded_Handle, swing1LimitDegrees=30, swing2LimitDegrees=30, twistLimitDegrees=5}
 {bone1Name=dyn_fish02_tFin, bone2Name=dyn_fish02_head, swing1LimitDegrees=15, swing2LimitDegrees=15, twistLimitDegrees=5}
 {bone1Name=dyn_fish02_bFin, bone2Name=dyn_fish02_head, swing1LimitDegrees=15, swing2LimitDegrees=15, twistLimitDegrees=5}
 {bone1Name=dyn_fish02_tail, bone2Name=Melee_TwoHanded_Handle, swing1LimitDegrees=60, swing2LimitDegrees=50, twistLimitDegrees=2}
 {bone1Name=dyn_fish03_tail, bone2Name=Melee_TwoHanded_Handle, swing1LimitDegrees=30, swing2LimitDegrees=30, twistLimitDegrees=15}
 {bone1Name=dyn_fish03_head, bone2Name=Melee_TwoHanded_Handle, swing1LimitDegrees=45, swing2LimitDegrees=45, twistLimitDegrees=6}
 {bone1Name=dyn_fish03_bFin, bone2Name=dyn_fish03_head, swing1LimitDegrees=15, swing2LimitDegrees=15, twistLimitDegrees=5}
 {bone1Name=dyn_fish03_tFin, bone2Name=dyn_fish03_head, swing1LimitDegrees=15, swing2LimitDegrees=15, twistLimitDegrees=5}
 {bone1Name=dyn_fish04_tail, bone2Name=Melee_TwoHanded_Handle, swing1LimitDegrees=50, swing2LimitDegrees=40, twistLimitDegrees=8}
 {bone1Name=dyn_fish04_head, bone2Name=Melee_TwoHanded_Handle, swing1LimitDegrees=50, swing2LimitDegrees=50, twistLimitDegrees=5}
 {bone1Name=dyn_fish04_tFin, bone2Name=dyn_fish04_head, swing1LimitDegrees=15, swing2LimitDegrees=15, twistLimitDegrees=5}
 {bone1Name=dyn_fish04_bFin, bone2Name=dyn_fish04_head, swing1LimitDegrees=15, swing2LimitDegrees=15, twistLimitDegrees=5}
]
